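# Material function in the NarrowFlea cliff terrain set that centralizes the grass stripe colors as shared parameters.
# It exposes three vector parameters in the NarrowFlea group and passes each RGB to a matching output: GrassStripeA, GrassStripeB and GrassStripeSSS.
# The A and B defaults are near-white, slightly green values that exceed one in some channels, while the SSS default is dark green.
# It takes no inputs and does no math.
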MaterialFunction /NarrowFleaEnv/Terrain/Cliffs/Materials/MaterialFunctions/MF_NarrowFlea_Colors
graph {nodes [
  FunctionOutput(None=VectorParameter() {group=NarrowFlea, parameterName=GrassStripeA, defaultValue=<rgba 0.9005 1.05 0.8528 1>}.RGB) {outputName=GrassStripeA}
  FunctionOutput(None=VectorParameter() {group=NarrowFlea, parameterName=GrassStripeB, defaultValue=<rgba 1.1327 1.328 1.0116 1>}.RGB) {outputName=GrassStripeB, sortPriority=1}
  FunctionOutput(None=VectorParameter() {group=NarrowFlea, parameterName=GrassStripeSSS, defaultValue=<rgba 0.0451 0.08 0.018 1>}.RGB) {outputName=GrassStripeSSS, sortPriority=2}
 ]}
